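# Translucent, unlit, two-sided Niagara lightning material for the WordUnity emote, tinted by particle color with depth fade.
# Scalar switches blend between noise-broken spike textures, a simple sprite and a panning spline trail.
# Dynamic parameters drive height-lerp dissolve, trail appearance, width and per-particle randomness.
# Ceil-quantized time makes panning jumpy, while trail world position offset combines SplineThicken with a sine wobble along WPODirection.
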Material /BRCosmetics/Animation/Game/MainPlayer/Emotes/WordUnity/FX/Materials/M_Emote_WordUnity_Lightning
graph {
 properties {
  bAutomaticallySetUsageInEditor=false
  bUsedWithNiagaraMeshParticles=true
  bUsedWithNiagaraRibbons=true
  bUsedWithNiagaraSprites=true
  blendMode=BLEND_Translucent
  shadingModel=MSM_Unlit
  twoSided=true
  usedShadingModels=Unlit
 }
 nodes [
  n1 = DynamicParameter() {parameterIndex=1, defaultValue=<rgba 0 0 0 0>, paramNames [Random,Random2,Param3,Param4]}
  n2 = Add(A=@n1.Random, B=Time() {bIgnorePause=true, bOverride_Period=true, period=6000})
  n3 = Multiply(B=@n1.Random2) {constA=20}
  n4 = Multiply(A=Ceil(None=Multiply(A=@n2, B=@n3) {constB=20})) {constB=0.8}
  Add(A=Add(A=Time(), B=@n1.Random), B=@n4)
  Add()
  DepthFade(Opacity=NamedRerouteUsage(), FadeDistance=ScalarParameter() {defaultValue=10, parameterName=FadeDistance}) {fadeDistanceDefault=10}
  n5 = ParticleColor()
  n6 = MaterialFunctionCall() {materialFunction=&Swizzle}
  n7 = Power(
   Base=StaticComponentMaskParameter(None=TextureSampleParameter2D(UVs=@n6.YX) {parameterName=TexSpikes, texture=/BRCosmetics/Effects/Fort_Effects/Emotes/ChimeCurlCorn/Materials/T_ChimeCurlCorn_Lightning}.RGBA) {defaultR=true, parameterName=TexSpikes_Mask}
   Exp=ScalarParameter() {defaultValue=1, parameterName=BasePower}
  ) {constExponent=1}
  n8 = TextureSample(UVs=Add(A=TextureCoordinate() {uTiling=0.5, vTiling=0.5}) {constB=0.2}) {samplerType=SAMPLERTYPE_LinearGrayscale, texture=/BRCosmetics/Effects/Fort_Effects/Emotes/ChimeCurlCorn/Materials/T_ChimeCurlCorn_Noise}
  n9 = Power(Base=LinearInterpolate(Alpha=@n8.R) {constA=0.1}, Exp=ScalarParameter() {defaultValue=1, parameterName=NoisePower})
  n10 = Multiply(A=@n5.A, B=Saturate(None=Multiply(A=@n7, B=@n9)))
  n11 = MaterialFunctionCall(
   Transition Phase=DynamicParameter() {defaultValue=<rgba 0 0 0 0>, paramNames [Dissolve,WidthBase,WidthTip,Appear]}.Dissolve
   Height Texture=ComponentMask(None=TextureCoordinate()) {g=true}
   Contrast=Constant() {r=2}
  ) {materialFunction=&HeightLerp}
  LinearInterpolate(
   A=LinearInterpolate(A=Multiply(A=@n10, B=@n11.Results), B=NamedRerouteUsage(), Alpha=ScalarParameter() {parameterName=UseSimpleSprite})
   Alpha=ScalarParameter() {parameterName=UseTrailSpline}
  )
  n12 = DynamicParameter() {defaultValue=<rgba 0 0 0 0>, paramNames [Dissolve,WidthBase,WidthTip,Appear]}
  n13 = StaticBool()
  n14 = TextureCoordinate()
  n15 = Saturate(None=Multiply(A=ComponentMask(None=@n14) {r=true}) {constB=5})
  n16 = TextureCoordinate() {uTiling=1.6}
  n17 = Multiply(A=Time() {bOverride_Period=true, period=6000}) {constB=3}
  n18 = Time() {bIgnorePause=true, bOverride_Period=true, period=6000}
  n19 = Multiply(A=Ceil(None=Multiply(A=@n18) {constB=10})) {constB=0.5}
  n20 = Sine(None=Add(A=ComponentMask(None=@n16) {r=true}, B=LinearInterpolate(A=@n17, B=@n19) {constAlpha=0.3})) {period=0.8}
  n21 = Multiply(
   A=LinearInterpolate(Alpha=Power(Base=@n20) {constExponent=3}) {constB=1.5}
   B=Transform(None=VectorParameter() {parameterName=WPODirection, defaultValue=<rgba 0 10 0 0>}.RGB)
  )
  LinearInterpolate(
   B=Add(
    A=MaterialFunctionCall(WidthBase=@n12.WidthBase, WidthTip=@n12.WidthBase, Expand U or V UV Channel=@n13) {materialFunction=&SplineThicken}.WorldPositionOffset
    B=Multiply(A=@n21, B=@n15)
   ) {constB=0}
   Alpha=ScalarParameter() {parameterName=UseTrailSpline}
  )
  MaterialFunctionCall(Emissive=@n5.RGB) {materialFunction=&MF_FX_EmissiveAdjustments}
  n22 = OneMinus(None=ComponentMask(None=TextureCoordinate()) {r=true})
  n23 = Multiply(B=Saturate(None=Multiply(A=@n22) {constB=10})) {constA=1}
  Multiply(A=@n12.WidthBase, B=@n23)
  Multiply(A=@n12.WidthTip, B=@n23)
  n24 = TextureCoordinate()
  n25 = AppendVector(A=ScalarParameter() {defaultValue=0.5, parameterName=UTiling}, B=Constant() {r=1})
  n26 = ScalarParameter() {defaultValue=-2, parameterName=PanSpeed}
  n27 = Constant2Vector() {r=1}
  n28 = TextureSampleParameter2D(UVs=Panner(Coordinate=Multiply(A=@n24, B=@n25), Time=@n4, Speed=Multiply(A=@n26, B=@n27)) {speedX=4}) {parameterName=TexTrail, texture=/BRCosmetics/Effects/Fort_Effects/Emotes/ChimeCurlCorn/Materials/T_ChimeCurlCorn_LightningTrail}
  n29 = Add(A=TextureCoordinate() {uTiling=0.5, vTiling=0.5}) {constB=0.3}
  n30 = Multiply(
   A=Multiply(A=ParticleColor().A, B=StaticComponentMaskParameter(None=@n28.RGBA) {defaultR=true, parameterName=TexTrail_Mask})
   B=MaterialFunctionCall(
    Transition Phase=DynamicParameter() {defaultValue=<rgba 0 0 0 0>, paramNames [Dissolve,WidthBase,WidthTip,Appear]}.Dissolve
    Height Texture=TextureSample(UVs=@n29) {samplerType=SAMPLERTYPE_Grayscale, texture=&LowResBlurredNoise}.R
    Contrast=Constant() {r=2}
   ) {materialFunction=&HeightLerp}.Results
  )
  n31 = OneMinus(None=ComponentMask(None=TextureCoordinate()) {r=true})
  n32 = Saturate(None=Multiply(A=Power(Base=@n31)) {constB=10})
  n33 = ComponentMask(None=TextureCoordinate()) {r=true}
  n34 = DynamicParameter() {defaultValue=<rgba 0 0 0 0>, paramNames [Dissolve,WidthBase,WidthTip,Appear]}
  n35 = Multiply(A=Add(A=OneMinus(None=@n33), B=LinearInterpolate(Alpha=@n34.Appear) {constA=-1}) {constB=0}) {constB=3}
  NamedRerouteDeclaration(None=Multiply(A=Multiply(A=@n30, B=@n32), B=Saturate(None=@n35))) {name=TrailSpline, nodeColor=<rgba 0.0333 0.0181 1 1>}
 ]
}
