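# Niagara system for the WordUnity emote that surrounds the character in a timed magenta glow effect.
# Burst emitters attach to skeletal mesh bones and sockets: eye glows, mesh body glows, anticipation and main flares, radial rays, cylinder-arranged lightning and a pulsing light.
# Curves over particle age or loop age drive size, alpha and material parameters.
# Front-end and time-of-day modifiers tune brightness, and the light spawns only in gameplay.
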NiagaraSystem /BRCosmetics/Animation/Game/MainPlayer/Emotes/WordUnity/FX/NS_Emote_WordUnity_BodyGlow
dependencies {
 usedDataInterfaces [
  &NiagaraDataInterfaceCurve
  &NiagaraDataInterfaceVector2DCurve
  &NiagaraDataInterfaceColorCurve
  &NiagaraDataInterfaceArrayFloat3
  &NiagaraDataInterfaceArrayInt32
  &NiagaraDataInterfaceSkeletalMesh
 ]
 usedDynamicInputs []
 usedModules [
  &self:SystemSpawnScript.SystemScriptSource.SystemScriptGraph.NiagaraNodeAssignment_6.SetVariables_#1
  &SystemState
 ]
}
emitters {
 AnticipationFlare {
  emitterUpdateScript [
   &EmitterState()
   &SpawnBurst_Instantaneous(Spawn Time=1.9, Use Spawn Probability=false, UseLoopCountLimit=false)
  ]
  particleSpawnScript [
   &InitializeParticle(
    Lifetime=1.7
    Position=<0 0 20>
    Position Mode=::NewEnumerator1
    Sprite Rotation Angle=0.25
    Sprite Rotation Mode=::NewEnumerator1
    Sprite Size Mode=::NewEnumerator3
    Uniform Sprite Size=1
    UsePositionOffset=false
    Write Material Random=true
   )
   &SkeletalMeshLocation(
    Bone / Socket Index=11
    Bone Sampling Mode=::NewEnumerator4
    MaskSpawnGroup=false
    Sampled Position Offset=<0 0 -50>
    Sampled Position Offset Coordinate Space:ENiagaraSkelMeshTransforms=::NewEnumerator1
    Skeletal Mesh=$System.Skeletal Mesh
   )
   &CameraOffset(Camera Offset Amount=50)
   &FN_SetSpawnGroup(Number of Spawn Groups=2)
  ]
  particleUpdateScript [
   &SolveForcesAndVelocity(Clamp Velocity=false, Limit Acceleration=false)
   &ParticleState()
   &ScaleSpriteSize(
    Non-Uniform Curve Scale=<600 200>
    Non-Uniform Curve Sprite Scale=DataInterface() {propertyValues {
      XCurve {keys [
        {interpMode=RCIM_Cubic, tangentMode=RCTM_User, time=0, value=0}
        {interpMode=RCIM_Cubic, tangentMode=RCTM_User, time=0.4, value=1}
        {tangentMode=RCTM_User, time=1, value=0}
       ]}
      YCurve {keys [
        {interpMode=RCIM_Cubic, tangentMode=RCTM_User, time=0, value=0}
        {arriveTangent=0.4878, interpMode=RCIM_Cubic, leaveTangent=0.4878, tangentMode=RCTM_User, time=0.8, value=1}
        {tangentMode=RCTM_User, time=1, value=0}
       ]}
     }}
    Scale Sprite Size Mode=::NewEnumerator3
   )
   &Color(Color=&ColorFromCurve(ColorCurve=DataInterface() {propertyValues {
       AlphaCurve {keys [<at 0 0>,<at 0.073 0.5 RCIM_Cubic>,<at 0.3152 0 RCIM_Cubic>]}
       BlueCurve {keys [<at 0 16.6666 RCIM_Cubic>]}
       GreenCurve {keys [<at 0 0 RCIM_Cubic>]}
       RedCurve {keys [<at 0 20 RCIM_Cubic>]}
      }}))
   /SetVariables(Particles.FlarePosition:NiagaraPosition=&ConvertVectorToPosition(
     Input Position=&Add_Vector(
      A=&ConvertPositionToVector(Input Position=$Particles.Position, Convert To Absolute Large World Vector=false)
      B=&MakeVector(
       X=0
       Y=0
       Z=&FloatFromCurve(
        FloatCurve=DataInterface() {propertyValues {Curve {keys [{tangentMode=RCTM_User, time=0, value=-0.8}, {tangentMode=RCTM_User, time=1, value=1}]}}}
        Scale Curve=-35
       )
      )
     )
     Vector Is In Absolute Large World Space:NiagaraBool=false
    )) {isSetParameters=true}
  ]
  rendererClasses [&NiagaraSpriteRendererProperties]
  renderers [{rendererClass=&NiagaraSpriteRendererProperties}]
 }
 EyeGlows {
  emitterUpdateScript [
   &EmitterState(Life Cycle Mode=::NewEnumerator1, Loop Behavior=::NewEnumerator1, Loop Duration=2, UseLoopDelay=false)
   &SpawnBurst_Instantaneous(Spawn Count=2, Spawn Time=0.25, Use Spawn Probability=false, UseLoopCountLimit=false)
  ]
  particleSpawnScript [
   &InitializeParticle(
    Color=<rgba 15 0 14.8698 1>
    Lifetime=2
    Sprite Size=<9 5>
    Sprite Size Mode=::NewEnumerator4
    UsePositionOffset=false
    Write Material Random=true
   )
   &PartitionParticles(ID to Partition:NiagaraInt32=&ReturnExecIndex())
   &CameraOffset(Camera Offset Amount=3)
  ]
  particleUpdateScript [
   &SkeletalMeshLocation(
    Bone / Socket Index=&SelectIntFromArray(Direct Array Index=$Particles.PartitionParticles.Partition, Int Selection Array=DataInterface() {propertyValues {IntData [0,1]}})
    MaskSpawnGroup=false
    Mesh Sampling Type=::NewEnumerator5
    Sampled Position Offset=&SelectVectorFromArray(
     Direct Array Index=$Particles.PartitionParticles.Partition
     Vector Selection Array=DataInterface() {propertyValues {FloatData [<4.6 8 4>, <4.6 8 -4>]}}
    )
    Skeletal Mesh=$System.Skeletal Mesh
    Socket Sampling Mode=::NewEnumerator5
   )
   &ParticleState()
   &ScaleColor(
    Scale Alpha=&FloatFromCurve(FloatCurve=DataInterface() {propertyValues {Curve {
        defaultValue=1
        keys [
         {interpMode=RCIM_Cubic, tangentMode=RCTM_User, time=0, value=0}
         {arriveTangent=1.1111, interpMode=RCIM_Cubic, leaveTangent=1.1111, time=0.3, value=1}
         {interpMode=RCIM_Cubic, tangentMode=RCTM_User, time=0.9, value=1}
         {tangentMode=RCTM_User, time=1, value=0}
        ]
       }}})
    Scale RGB=&MakeVectorFromLinearColorRGB(Color=<rgba 3 3 3 1>)
    ScaleA=true
    ScaleRGB=true
   )
   /SetVariables(Particles.VisibilityTag=$Particles.PartitionParticles.Partition) {enabled=false, isSetParameters=true}
   /SetVariables(
    Particles.LargerGlowColor:LinearColor=&Multiply_LinearColorByFloat(Linear Color=$Particles.Color, Float=0.15)
    Particles.LargerGlowSize:Vector2f=&Multiply_Vector2D_ByFloat(Float=3, Vector2D=$Particles.SpriteSize)
   ) {isSetParameters=true}
  ]
  rendererClasses [&NiagaraSpriteRendererProperties]
  renderers [
   {rendererClass=&NiagaraSpriteRendererProperties}
   {rendererClass=&NiagaraSpriteRendererProperties, rendererIndex=1}
  ]
 }
 Flares {
  emitterUpdateScript [
   &EmitterState()
   &SpawnBurst_Instantaneous(Spawn Time=1, Use Spawn Probability=false, UseLoopCountLimit=false)
  ]
  particleSpawnScript [
   &InitializeParticle(
    Color=<rgba 3 0.15 3 0.5>
    Lifetime=2.25
    Sprite Size=<240 360>
    Sprite Size Mode=::NewEnumerator4
    UsePositionOffset=false
    Write Material Random=true
   )
   &CameraOffset(Camera Offset Amount=-20)
   &DynamicMaterialParameters(
    Index 0 Param 2=0.2
    Module.Param0WriteEnabled=true
    Module.Param1WriteEnabled=true
    Module.Param2WriteEnabled=true
    Module.Param3WriteEnabled=true
   )
  ]
  particleUpdateScript [
   &ParticleState()
   &SkeletalMeshLocation(
    Bone / Socket Index=10
    Bone Sampling Mode=::NewEnumerator4
    MaskSpawnGroup=false
    Orientation Sampling=::NewEnumerator1
    Sampled Position Offset=<-20 0 0>
    Skeletal Mesh=$System.Skeletal Mesh
   )
   &ScaleColor(
    Scale Alpha=&FloatFromCurve(
     FloatCurve=DataInterface() {propertyValues {Curve {
        defaultValue=0
        keys [
         {arriveTangent=9.0725, interpMode=RCIM_Cubic, leaveTangent=9.0725, tangentMode=RCTM_User, time=0, value=0}
         {arriveTangent=-2.5249, interpMode=RCIM_Cubic, leaveTangent=-2.5249, tangentMode=RCTM_User, time=0.1408, value=0.9847}
         {arriveTangent=-0.0405, interpMode=RCIM_Cubic, leaveTangent=-0.0405, tangentMode=RCTM_User, time=1, value=0}
        ]
       }}}
     CurveIndex=$System.NormalizedLoopAge
     Scale Curve=3
    )
    Scale RGB=&Multiply_VectorByFloat(
     Vector=&VectorFromFloat(Value=&FNTimeOfDayScalar(Night Scale=0.6, Morning Scale=0.6, Evening Scale=0.6))
     Float=&FN_FrontEndFloatModifier(Gameplay=1, Front End=0.75)
    )
    ScaleA=true
    ScaleRGB=true
   )
   &ScaleSpriteSize(Uniform Curve Sprite Scale=DataInterface() {propertyValues {Curve {keys [
        {arriveTangent=2.0725, interpMode=RCIM_Cubic, leaveTangent=2.0725, tangentMode=RCTM_User, time=0, value=0}
        {arriveTangent=1.3333, interpMode=RCIM_Cubic, leaveTangent=1.3333, time=0.1561, value=0.9952}
        {arriveTangent=-1.1792, interpMode=RCIM_Cubic, leaveTangent=-1.1792, time=0.75, value=1}
        <at 1 0 RCIM_Cubic>
       ]}}})
   /SetVariables(
    Particles.Color_Scale:LinearColor=&MakeLinearColorFromVectorAndFloat(
     "Vector (RGB)"=&Multiply_Vector(
      A=&MakeVectorFromLinearColorRGB(Color=<rgba 2 0.1 2 0.5>)
      B=&Multiply_VectorByFloat(
       Vector=&VectorFromFloat(Value=&FNTimeOfDayScalar(Night Scale=0.5, Morning Scale=0.25, Evening Scale=0.5))
       Float=&FN_FrontEndFloatModifier(Gameplay=1, Front End=0.75)
      )
     )
     "Float (Alpha)"=&FloatFromCurve(
      FloatCurve=DataInterface() {propertyValues {Curve {keys [
          {arriveTangent=1.9817, interpMode=RCIM_Cubic, leaveTangent=1.9817, tangentMode=RCTM_User, time=0, value=0}
          {arriveTangent=1.25, interpMode=RCIM_Cubic, leaveTangent=1.25, time=0.5, value=1}
          {arriveTangent=-0.4595, interpMode=RCIM_Cubic, leaveTangent=-0.4595, tangentMode=RCTM_User, time=0.8, value=1}
          {arriveTangent=-26.3419, interpMode=RCIM_Cubic, leaveTangent=-26.3419, tangentMode=RCTM_User, time=0.9, value=0}
         ]}}}
      CurveIndex=$System.NormalizedLoopAge
     )
    )
    Particles.Sprite_Size:Vector2f=&Multiply_Vector2D_ByFloat(
     Float=&Multiply_Float(
      A=&FloatFromCurve(FloatCurve=DataInterface() {propertyValues {Curve {keys [
           {arriveTangent=5.5466, interpMode=RCIM_Cubic, leaveTangent=5.5466, tangentMode=RCTM_User, time=0, value=0}
           {arriveTangent=-0.0043, interpMode=RCIM_Cubic, leaveTangent=-0.0043, tangentMode=RCTM_User, time=0.5, value=0.9762}
           {arriveTangent=-0.1074, interpMode=RCIM_Cubic, leaveTangent=-0.1074, tangentMode=RCTM_User, time=0.9, value=1}
           {arriveTangent=-21.0911, interpMode=RCIM_Cubic, leaveTangent=-21.0911, tangentMode=RCTM_User, time=1, value=0}
          ]}}})
      B=&RandomRangeFloat(Minimum=1.2, Maximum=1.5)
     )
     Vector2D=<200 180>
    )
   ) {isSetParameters=true}
   /SetVariables(
    Particles.Color01_Scale:LinearColor=&MakeLinearColorFromVectorAndFloat(
     "Vector (RGB)"=&MakeVectorFromLinearColorRGB(Color=<rgba 5 0.1 5 1>)
     "Float (Alpha)"=&FloatFromCurve(
      FloatCurve=DataInterface() {propertyValues {Curve {
         defaultValue=0
         keys [
          {arriveTangent=4.8768, interpMode=RCIM_Cubic, leaveTangent=4.8768, tangentMode=RCTM_User, time=0, value=0}
          {arriveTangent=-2.5249, interpMode=RCIM_Cubic, leaveTangent=-2.5249, tangentMode=RCTM_User, time=0.5, value=0.9847}
          {arriveTangent=-0.0405, interpMode=RCIM_Cubic, leaveTangent=-0.0405, tangentMode=RCTM_User, time=1, value=0}
         ]
        }}}
      Scale Curve=&FN_FrontEndFloatModifier(Gameplay=1, Front End=0.75)
     )
    )
    Particles.Sprite01_Size:Vector2f=&Vector2DFromFloat(Value=&FloatFromCurve(
      FloatCurve=DataInterface() {propertyValues {Curve {
         defaultValue=0
         keys [
          {arriveTangent=4.8768, interpMode=RCIM_Cubic, leaveTangent=4.8768, tangentMode=RCTM_User, time=0, value=0}
          {arriveTangent=-2.5249, interpMode=RCIM_Cubic, leaveTangent=-2.5249, tangentMode=RCTM_User, time=0.5, value=0.9847}
          {arriveTangent=-0.0405, interpMode=RCIM_Cubic, leaveTangent=-0.0405, tangentMode=RCTM_User, time=1, value=0}
         ]
        }}}
      Scale Curve=175
     ))
   ) {isSetParameters=true}
  ]
  rendererClasses [&NiagaraSpriteRendererProperties]
  renderers [
   {rendererClass=&NiagaraSpriteRendererProperties}
   {rendererClass=&NiagaraSpriteRendererProperties, rendererIndex=1}
   {rendererClass=&NiagaraSpriteRendererProperties, rendererIndex=2}
  ]
 }
 Glow_Body002 {
  emitterUpdateScript [
   /SetVariables(Emitter.OutlineThickness:NiagaraFloat=&FloatFromCurve(
     FloatCurve=DataInterface() {propertyValues {Curve {keys [
         {arriveTangent=13.6388, interpMode=RCIM_Cubic, time=0.003, value=0.0083}
         {arriveTangent=15.3759, leaveTangent=15.3759, time=0.1, value=1.5}
         <at 0.3 1.5>
         {interpMode=RCIM_Cubic, tangentMode=RCTM_User, time=0.9, value=1.5}
         {interpMode=RCIM_Cubic, leaveTangent=-1.6667, time=1, value=0}
        ]}}}
     CurveIndex=$System.NormalizedLoopAge
     Scale Curve=2
    )) {isSetParameters=true}
   /SetVariables(Emitter.GlowOutlineThickness:NiagaraFloat=&Multiply_Float(A=$Emitter.OutlineThickness, B=2)) {isSetParameters=true}
   &EmitterState(Life Cycle Mode=::NewEnumerator1, Loop Behavior=::NewEnumerator1, Loop Duration=1, UseLoopDelay=false)
   &SpawnBurst_Instantaneous(Spawn Count=7, Spawn Time=0.2, Use Spawn Probability=false, UseLoopCountLimit=false)
  ]
  particleSpawnScript [
   &InitializeParticle(Lifetime=4, Mesh Scale Mode=::NewEnumerator4, UsePositionOffset=true, Write Material Random=true)
   &PartitionParticles(Number Of Partitions=7)
   /SetVariables() {enabled=false, isSetParameters=true}
   /SetVariables(Particles.VisibilityTag=&MakeIntFromFloat(Float=&FN_FrontEndFloatModifier(Gameplay=0, Front End=1))) {isSetParameters=true}
  ]
  particleUpdateScript [
   &ParticleState()
   &ScaleMeshSize(Scale Factor=&VectorFromFloat(Value=0.8))
   &ScaleColor(
    Scale Alpha=&FloatFromCurve(
     FloatCurve=DataInterface() {propertyValues {Curve {keys [
         {arriveTangent=13.6388, interpMode=RCIM_Cubic, time=0.003, value=0.0083}
         {arriveTangent=15.3759, leaveTangent=15.3759, time=0.1, value=1.5}
         <at 0.3 1.5>
         {interpMode=RCIM_Cubic, tangentMode=RCTM_User, time=0.9, value=1.5}
         {interpMode=RCIM_Cubic, leaveTangent=-1.6667, time=1, value=0}
        ]}}}
     Scale Curve=&MakeCustomFloatFromBool(Bool=$User.bIsFrontEndPreview, True Float=30, False Float=20)
    )
    ScaleA=true
    ScaleRGB=true
   ) {module=ScaleColor001}
   &SkeletalMeshLocation(
    Bone / Socket Index=$Particles.PartitionParticles.Partition
    Bone Sampling Mode=::NewEnumerator4
    MaskSpawnGroup=false
    Orientation Sampling=::NewEnumerator1
    Skeletal Mesh=$System.Skeletal Mesh
   )
   &CameraOffset(Camera Offset Amount=20)
   &DynamicMaterialParameters(Module.Index0Float0Write=true, Module.Index0Float1Write=true, Module.Index0Float2Write=true, Module.Index0Float3Write=true)
  ]
  rendererClasses [&NiagaraMeshRendererProperties]
  renderers [
   {rendererClass=&NiagaraMeshRendererProperties, rendererIndex=0}
   {rendererClass=&NiagaraMeshRendererProperties, rendererIndex=1}
   {rendererClass=&NiagaraMeshRendererProperties, rendererIndex=2}
   {rendererClass=&NiagaraMeshRendererProperties, rendererIndex=3}
  ]
 }
 Light {
  emitterUpdateScript [
   &EmitterState(Life Cycle Mode=::NewEnumerator1, Loop Behavior=::NewEnumerator1, Loop Duration=1, UseLoopDelay=false)
   &SpawnBurst_Instantaneous(
    Spawn Probability=&FN_FrontEndFloatModifier(Gameplay=1, Front End=0)
    Use Spawn Probability=true
    UseLoopCountLimit=false
   )
  ]
  particleSpawnScript [&InitializeParticle(Color=<rgba 0.9225 0.07 1 3>, Lifetime=4, UsePositionOffset=true, Write Material Random=true)]
  particleUpdateScript [
   &ParticleState()
   &ScaleColor(
    Scale Alpha=&FloatFromCurve(
     FloatCurve=DataInterface() {propertyValues {Curve {
        defaultValue=0.0083
        keys [
         {interpMode=RCIM_Cubic, tangentMode=RCTM_User, time=0.05, value=0}
         {arriveTangent=0.4982, interpMode=RCIM_Cubic, leaveTangent=0.4982, time=0.08, value=0.05}
         {arriveTangent=1.039, interpMode=RCIM_Cubic, leaveTangent=1.039, time=0.2507, value=0.1}
         {arriveTangent=0.0826, interpMode=RCIM_Cubic, leaveTangent=0.0826, tangentMode=RCTM_User, time=0.2725, value=0.25}
         {arriveTangent=0.3348, interpMode=RCIM_Cubic, leaveTangent=0.3348, tangentMode=RCTM_User, time=0.529, value=0.3}
         {arriveTangent=-5.6099, interpMode=RCIM_Cubic, leaveTangent=-5.6099, tangentMode=RCTM_User, time=0.5805, value=1.0722}
         {arriveTangent=-1.4282, interpMode=RCIM_Cubic, leaveTangent=-1.4282, tangentMode=RCTM_User, time=0.6691, value=0.2598}
         {tangentMode=RCTM_User, time=1, value=0}
        ]
       }}}
     Scale Curve=&MakeCustomFloatFromBool(Bool=$User.bIsFrontEndPreview, True Float=50, False Float=30)
    )
    ScaleA=true
    ScaleRGB=true
   ) {module=ScaleColor001}
   &SkeletalMeshLocation(
    Bone / Socket Index=10
    Bone Sampling Mode=::NewEnumerator4
    MaskSpawnGroup=false
    Orientation Sampling=::NewEnumerator1
    Sampled Position Offset=<-100 0 0>
    Skeletal Mesh=$System.Skeletal Mesh
   )
  ]
  rendererClasses [&NiagaraLightRendererProperties]
  renderers [{rendererClass=&NiagaraLightRendererProperties, rendererIndex=0}]
 }
 Lightning02 {
  emitterUpdateScript [
   &EmitterState(Life Cycle Mode=::NewEnumerator1, Loop Behavior=::NewEnumerator1, Loop Duration=2, UseLoopDelay=false)
   &SpawnBurst_Instantaneous(Spawn Count=18, Spawn Time=1, Use Spawn Probability=false, UseLoopCountLimit=false)
  ]
  particleSpawnScript [
   &InitializeParticle(
    Color=<rgba 10 0.7826 9.4391 1>
    Lifetime=2.4
    Mesh Scale Max=<1.5 1.5 1.5>
    Mesh Scale Mode=::NewEnumerator2
    Position Mode=::NewEnumerator1
    UsePositionOffset=false
    Write Material Random=true
   )
   &ShapeLocation(
    Cylinder Distribution=::NewEnumerator2
    Cylinder Height=&RandomRangeFloat(Minimum=0, Maximum=0)
    Cylinder Height Midpoint=0
    Cylinder Radius=&RandomRangeFloat(Minimum=5, Maximum=5)
    MaskSpawnGroup=false
    Non Uniform Scale=<1 1.5 1>
    Offset Coordinate Space=::Local
    Offset Mode=::NewEnumerator0
    Rotation Angle=0.25
    Rotation Coordinate Space=::Simulation
    Rotation Mode=::NewEnumerator1
    Shape Primitive=::NewEnumerator1
    U Position=&ReturnNormalizedExecIndex(Normalized Index Includes One=false)
    UseOwnerScale=true
   )
   &InitialMeshOrientation(Mesh Orientation Mode=::NewEnumerator2, Orientation Vector=$Particles.ShapeLocation.ShapeNormal, Use Rotation Vector=false)
   &CameraOffset(Camera Offset Amount=-20)
  ]
  particleUpdateScript [
   &ParticleState()
   &SkeletalMeshLocation(
    Bone / Socket Index=10
    Bone Sampling Mode=::NewEnumerator4
    MaskSpawnGroup=false
    Orientation Sampling=::NewEnumerator1
    Sampled Position Offset=<-20 0 0>
    Skeletal Mesh=$System.Skeletal Mesh
   )
   &ShapeLocation(MaskSpawnGroup=false, Override Seed=false, Shape Origin=$Particles.Position, Sphere Radius=6, UseOwnerScale=true) {module=ShapeLocation002}
   &ScaleMeshSize(Scale Factor=&VectorFromFloat(Value=&Multiply_Float(
      A=&Lerp_Float(
       A=0.7
       B=1
       Alpha=&FloatFromCurve(
        FloatCurve=DataInterface() {propertyValues {Curve {keys [<at 0 0>,<at 0.25 1.2>,<at 0.5 0>,<at 0.75 1>,<at 1 0>]}}}
        CurveIndex=&ReturnNormalizedExecIndex(Normalized Index Includes One=false)
       )
      )
      B=1.5
     ))) {module=ScaleMeshSize001}
   &DynamicMaterialParameters(
    Index 0 Param 1=&FloatFromCurve(FloatCurve=DataInterface() {propertyValues {Curve {
        defaultValue=0
        keys [
         {tangentMode=RCTM_User, time=0, value=0}
         <at 0.5159 0.6424>
         <at 0.7018 0.6518>
         <at 0.7318 0.6894>
         <at 0.9218 0.6988>
         {tangentMode=RCTM_User, time=1, value=0.8}
        ]
       }}})
    Index 0 Param 2=40
    Index 0 Param 3=&FloatFromCurve(
     FloatCurve=DataInterface() {propertyValues {Curve {
        defaultValue=0
        keys [
         {tangentMode=RCTM_User, time=0.484, value=1}
         <at 0.5159 0.6075>
         <at 0.7018 0.5981>
         <at 0.7318 0.5604>
         <at 0.9218 0.551>
         {tangentMode=RCTM_User, time=1, value=0.4498}
        ]
       }}}
     Scale Curve=40
    )
    Index 0 Param 4=&FloatFromCurve(
     FloatCurve=DataInterface() {propertyValues {Curve {
        defaultValue=0
        keys [
         {arriveTangent=5.2125, interpMode=RCIM_Cubic, leaveTangent=5.2125, tangentMode=RCTM_User, time=0, value=0}
         {arriveTangent=0.9556, interpMode=RCIM_Cubic, leaveTangent=0.9556, tangentMode=RCTM_User, time=0.2598, value=0.6352}
         {interpMode=RCIM_Cubic, tangentMode=RCTM_User, time=0.8, value=1}
        ]
       }}}
     CurveIndex=&Multiply_Float(A=$Particles.NormalizedAge, B=&RandomRangeFloat(Minimum=0.6, Maximum=1))
    )
    Index 1 Param 1=&RandomRangeFloat(Minimum=0, Maximum=2)
    Index 1 Param 2=&RandomRangeFloat(Minimum=0.9, Maximum=1.1)
    Index 1 Param 4=&FloatFromCurve(FloatCurve=DataInterface() {propertyValues {Curve {defaultValue=1, keys [{tangentMode=RCTM_User, time=0, value=0}, {tangentMode=RCTM_User, time=0.3, value=1}]}}})
    Module.Index0Float0Write=true
    Module.Index0Float1Write=true
    Module.Index0Float2Write=true
    Module.Index0Float3Write=true
    Module.Index1Float0Write=true
    Module.Index1Float1Write=true
    Module.Index1Float2Write=true
    Module.Index1Float3Write=true
    Write Parameter Index 1=::NewEnumerator1
   )
   &FN_Sprite_Attributes(
    Color=&MakeLinearColorFromVectorAndFloat(
     "Vector (RGB)"=&Multiply_VectorByFloat(Vector=&MakeVectorFromLinearColorRGB(Color=<rgba 1 0 1 1>), Float=1)
     "Float (Alpha)"=&Multiply_Float(
      A=&FNTimeOfDayScalar(Night Scale=0.4, Morning Scale=0.7, Evening Scale=0.7)
      B=&FN_FrontEndFloatModifier(Gameplay=1, Front End=0.5)
     )
    )
    Create Color Attribute=true
    Write to Module Namespace=true
   )
   &SolveForcesAndVelocity(Clamp Velocity=false, Limit Acceleration=false)
   &ScaleColor(Scale Alpha=&FN_FrontEndFloatModifier(Gameplay=1, Front End=0.5), ScaleA=true, ScaleRGB=true)
  ]
  rendererClasses [&NiagaraMeshRendererProperties]
  renderers [
   {rendererClass=&NiagaraMeshRendererProperties, rendererIndex=0}
   {rendererClass=&NiagaraMeshRendererProperties, rendererIndex=1}
  ]
 }
 Rays {
  emitterUpdateScript [
   &EmitterState()
   &SpawnBurst_Instantaneous(Spawn Count=8, Spawn Time=1, Use Spawn Probability=false, UseLoopCountLimit=false)
  ]
  particleSpawnScript [
   &PartitionParticles(Number Of Partitions=8)
   &InitializeParticle(
    Lifetime=2.5
    Mesh Scale Mode=::NewEnumerator3
    Position Offset=<0 0 120>
    Position Offset Coordinate Space=::Local
    Sprite Rotation Mode=::NewEnumerator2
    Sprite Size Mode=::NewEnumerator3
    Uniform Sprite Size=100
    UsePositionOffset=true
    Write Material Random=true
   )
   &CameraOffset(Camera Offset Amount=-20)
   &InitialMeshOrientation(
    Mesh Orientation Mode=::NewEnumerator2
    Rotation=&SelectVectorFromArray(
     Direct Array Index=$Particles.PartitionParticles.Partition
     Vector Selection Array=DataInterface() {propertyValues {FloatData [<0.125 0 0>, <0.25 0 0>, <0.375 0 0>, <0.5 0 0>, <0.625 0 0>, <0.75 0 0>, <0.875 0 0>, <0 0 0>]}}
    )
    Use Rotation Vector=true
   )
   &DynamicMaterialParameters(
    Index 0 Param 1=-0.5
    Module.Index0Param0Write=true
    Module.Index0Param1Write=true
    Module.Index0Param2Write=true
    Module.Index0Param3Write=true
   )
  ]
  particleUpdateScript [
   &ParticleState()
   &SkeletalMeshLocation(
    Bone / Socket Index=10
    Bone Sampling Mode=::NewEnumerator4
    MaskSpawnGroup=false
    Orientation Sampling=::NewEnumerator1
    Sampled Position Offset=<-20 0 0>
    Skeletal Mesh=$System.Skeletal Mesh
   )
   &ScaleMeshSize(Scale Factor=&VectorFromFloat(Value=&FloatFromCurve(
      FloatCurve=DataInterface() {propertyValues {Curve {
         defaultValue=1
         keys [
          {arriveTangent=1.1179, interpMode=RCIM_Cubic, leaveTangent=1.1179, tangentMode=RCTM_User, time=0, value=0.6228}
          {arriveTangent=0.0224, interpMode=RCIM_Cubic, leaveTangent=0.0224, tangentMode=RCTM_User, time=1, value=1}
         ]
        }}}
      Scale Curve=&Multiply_Float(
       A=1.7
       B=&MakeFloatFromVector(
        Vector=&VectorFromFloat(Value=&Lerp_Float(
          A=0.8
          B=1
          Alpha=&FloatFromCurve(
           FloatCurve=DataInterface() {propertyValues {Curve {keys [<at 0 0>,<at 0.25 1>,<at 0.5 0>,<at 0.75 1>,<at 1 0>]}}}
           CurveIndex=&Add_Float(A=&ReturnNormalizedExecIndex(Normalized Index Includes One=false), B=-0.125)
          )
         ))
        Channel=::NewEnumerator0
       )
      )
     )))
   &ScaleColor(
    Scale Alpha=&FloatFromCurve(
     FloatCurve=DataInterface() {propertyValues {Curve {
        defaultValue=0
        keys [
         {interpMode=RCIM_Cubic, tangentMode=RCTM_User, time=0, value=0}
         <at 0.5 1.0147>
         {arriveTangent=-0.0405, interpMode=RCIM_Cubic, leaveTangent=-0.0405, tangentMode=RCTM_User, time=0.6, value=1}
         {arriveTangent=-0.1782, interpMode=RCIM_Cubic, leaveTangent=-0.1782, tangentMode=RCTM_User, time=1, value=0}
        ]
       }}}
     Scale Curve=&FN_FrontEndFloatModifier(Gameplay=2, Front End=1)
    )
    Scale RGB=&Multiply_VectorByFloat(
     Vector=&MakeVectorFromLinearColorRGB(Color=&ColorFromCurve(ColorCurve=DataInterface() {propertyValues {
         AlphaCurve {keys [<at 0 1 RCIM_Cubic>]}
         BlueCurve {keys [<at 0.0211 10>, {arriveTangent=-9.194, interpMode=RCIM_Cubic, leaveTangent=-9.194, time=0.4003, value=10}, <at 1 1>]}
         GreenCurve {keys [<at 0.0211 0>,<at 0.4003 0.0705 RCIM_Cubic>,<at 1 0>]}
         RedCurve {keys [<at 0.0211 10>, {arriveTangent=-9.194, interpMode=RCIM_Cubic, leaveTangent=-9.194, time=0.4003, value=10}, <at 1 1>]}
        }}))
     Float=&FloatFromCurve(FloatCurve=DataInterface() {propertyValues {Curve {
         defaultValue=0
         keys [
          {tangentMode=RCTM_User, time=0, value=1}
          <at 0.1 1.0053>
          {arriveTangent=-0.0405, interpMode=RCIM_Cubic, leaveTangent=-0.0405, tangentMode=RCTM_User, time=0.7009, value=1}
          {arriveTangent=-8.0089, interpMode=RCIM_Cubic, leaveTangent=-8.0089, tangentMode=RCTM_User, time=1, value=0}
         ]
        }}})
    )
    ScaleA=true
    ScaleRGB=true
   ) {module=ScaleColor001}
   /SetVariables(Particles.Color_02:LinearColor=&MultiplyLinearColor(
     Linear Color=$Particles.Color
     Scale Factor=&MakeVector4FromVectorAndFloat("Vector (RGB)"=<1 1 1>, "Float (Alpha)"=0.66)
    )) {isSetParameters=true}
  ]
  rendererClasses [&NiagaraMeshRendererProperties]
  renderers [
   {rendererClass=&NiagaraMeshRendererProperties, rendererIndex=0}
   {rendererClass=&NiagaraMeshRendererProperties, rendererIndex=1}
  ]
 }
}
summary {
 systemName=NS_Emote_WordUnity_BodyGlow
 emitters [
  {emitterName=Glow_Body002, rendererClasses [&NiagaraMeshRendererProperties]}
  {emitterName=Lightning02, rendererClasses [&NiagaraMeshRendererProperties]}
  {emitterName=Light, rendererClasses [&NiagaraLightRendererProperties]}
  {emitterName=AnticipationFlare, rendererClasses [&NiagaraSpriteRendererProperties]}
  {emitterName=Flares, rendererClasses [&NiagaraSpriteRendererProperties]}
  {emitterName=Rays, rendererClasses [&NiagaraMeshRendererProperties]}
  {emitterName=EyeGlows, rendererClasses [&NiagaraSpriteRendererProperties]}
 ]
 userVariables [
  User.bIsFrontEnd {type=NiagaraBool}
  User.bIsFrontEndPreview {type=NiagaraBool}
  User.bisHeadHidden {type=NiagaraBool}
 ]
}
