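# Niagara burst effect for the return transformation of the WordUnity emote on the main player, with magenta HDR sprites that fade quickly.
# NE_Decals spawns one burst, assigns each sprite a bone index from its execution index, and follows that socket through a system Socket Reader.
# NE_Flare adds a raised, camera-offset flare drawn by three sprite renderers, using module-written color and size attributes, with one size scaled differently for gameplay and front end.
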NiagaraSystem /BRCosmetics/Animation/Game/MainPlayer/Emotes/WordUnity/FX/NS_Emote_WordUnity_Transformation_Return
dependencies {
 usedDataInterfaces [&NiagaraDataInterfaceCurve, &NiagaraDataInterfaceColorCurve, &NiagaraDataInterfaceSocketReader]
 usedDynamicInputs []
 usedModules [
  &self:SystemSpawnScript.SystemScriptSource.SystemScriptGraph.NiagaraNodeAssignment_2.SetVariables_#1
  &SystemState
 ]
}
emitters {
 NE_Decals {
  emitterUpdateScript [
   &EmitterState()
   &SpawnBurst_Instantaneous(Spawn Count=15, Use Spawn Probability=false, UseLoopCountLimit=false)
  ]
  particleSpawnScript [
   &InitializeParticle(
    Lifetime=0.5
    Sprite Rotation Mode=::NewEnumerator2
    Sprite Size Mode=::NewEnumerator3
    Uniform Sprite Size=100
    UsePositionOffset=true
    Write Material Random=true
   )
   &CameraOffset(Camera Offset Amount=60)
   /SetVariables(Particles.BoneIndex:NiagaraInt32=&ReturnExecIndex()) {isSetParameters=true}
   &DynamicMaterialParameters(
    Index 0 Param 1=-0.5
    Module.Index0Param0Write=true
    Module.Index0Param1Write=true
    Module.Index0Param2Write=true
    Module.Index0Param3Write=true
   )
  ]
  particleUpdateScript [
   &ParticleState()
   &SocketLocation(
    SamplePosOffset=false
    Socket Index=$Particles.BoneIndex
    Socket Reader=$System.Socket Reader
    Socket Sampling Mode=::NewEnumerator3
    VelocitySpeedLimit=false
   )
   &ScaleColor(
    Scale Alpha=&FloatFromCurve(FloatCurve=DataInterface() {propertyValues {Curve {
        defaultValue=0
        keys [{interpMode=RCIM_Cubic, tangentMode=RCTM_User, time=0, value=1}, {interpMode=RCIM_Cubic, tangentMode=RCTM_User, time=1, value=0}]
       }}})
    Scale RGB=&Multiply_VectorByFloat(
     Vector=&MakeVectorFromLinearColorRGB(Color=&ColorFromCurve(ColorCurve=DataInterface() {propertyValues {
         AlphaCurve {keys [<at 0 1 RCIM_Cubic>]}
         BlueCurve {keys [<at 0.0211 10>, {arriveTangent=-9.194, interpMode=RCIM_Cubic, leaveTangent=-9.194, time=0.4003, value=10}, <at 1 1>]}
         GreenCurve {keys [<at 0.0211 0>,<at 0.4003 0.0705 RCIM_Cubic>,<at 1 0>]}
         RedCurve {keys [<at 0.0211 10>, {arriveTangent=-9.194, interpMode=RCIM_Cubic, leaveTangent=-9.194, time=0.4003, value=10}, <at 1 1>]}
        }}))
     Float=&FloatFromCurve(FloatCurve=DataInterface() {propertyValues {Curve {
         defaultValue=0
         keys [{interpMode=RCIM_Cubic, tangentMode=RCTM_User, time=0, value=1}, {interpMode=RCIM_Cubic, tangentMode=RCTM_User, time=1, value=0}]
        }}})
    )
    ScaleA=true
    ScaleRGB=true
   ) {module=ScaleColor001}
  ]
  rendererClasses [&NiagaraSpriteRendererProperties]
  renderers [{rendererClass=&NiagaraSpriteRendererProperties}]
 }
 NE_Flare {
  emitterUpdateScript [
   &EmitterState()
   &SpawnBurst_Instantaneous(Use Spawn Probability=false, UseLoopCountLimit=false)
  ]
  particleSpawnScript [
   &InitializeParticle(
    Color Mode=::NewEnumerator0
    Lifetime=&MaskFloatBySpawnGroup(Masked Float=0.6, Passthrough Float=1)
    Mesh Scale Mode=::NewEnumerator3
    Position Offset=<0 0 120>
    Sprite Size=<100 150>
    Sprite Size Mode=::NewEnumerator4
    UsePositionOffset=true
    Write Material Random=true
   )
   &CameraOffset(Camera Offset Amount=150)
  ]
  particleUpdateScript [
   &ParticleState()
   &ScaleSpriteSize(Uniform Curve Sprite Scale=DataInterface() {propertyValues {Curve {keys [
        {arriveTangent=1.4878, interpMode=RCIM_Cubic, leaveTangent=1.4878, tangentMode=RCTM_User, time=0, value=0.25}
        {interpMode=RCIM_Cubic, tangentMode=RCTM_User, time=0.25, value=1}
        <at 1 0 RCIM_Cubic>
       ]}}})
   &Color(Color=&ColorFromCurve(ColorCurve=DataInterface() {propertyValues {
       AlphaCurve {keys [<at 0.5878 1 RCIM_Cubic>,<at 0.9954 0>]}
       BlueCurve {keys [<at 0.0522 35>, {arriveTangent=-35.872, interpMode=RCIM_Cubic, leaveTangent=-35.872, time=0.1651, value=30}, <at 1 1 RCIM_Cubic>]}
       GreenCurve {keys [<at 0.0522 0>, {arriveTangent=0.0221, interpMode=RCIM_Cubic, leaveTangent=0.0221, time=0.1651, value=0}, <at 1 0.021 RCIM_Cubic>]}
       RedCurve {keys [<at 0.0522 35>, {arriveTangent=-35.872, interpMode=RCIM_Cubic, leaveTangent=-35.872, time=0.1651, value=30}, <at 1 1 RCIM_Cubic>]}
      }}))
   &FN_Sprite_Attributes(
    Color=&MultiplyLinearColor(Linear Color=$Particles.Color, Scale Factor=&MakeVector4_FromLinearColor(LinearColor=<rgba 1 1 1 1>))
    Create Color Attribute=true
    Create SpriteSize Attribute=true
    SpriteSize=&Multiply_Vector2D_ByFloat(Float=1, Vector2D=$Particles.SpriteSize)
    Write to Module Namespace=true
   )
   &FN_Sprite_Attributes(
    Color=&ColorFromCurve(ColorCurve=DataInterface() {propertyValues {
       AlphaCurve {keys [<at 0 0.25 RCIM_Cubic>,<at 0.8324 0 RCIM_Cubic>]}
       BlueCurve {keys [<at 0 6 RCIM_Cubic>, {interpMode=RCIM_Cubic, leaveTangent=-7.0008, time=0.8393, value=1}]}
       GreenCurve {keys [<at 0 0 RCIM_Cubic>, {interpMode=RCIM_Cubic, leaveTangent=0.4199, time=0.8393, value=0}]}
       RedCurve {keys [<at 0 6 RCIM_Cubic>, {interpMode=RCIM_Cubic, leaveTangent=-5.834, time=0.8393, value=1}]}
      }})
    Create Color Attribute=true
    Create SpriteSize Attribute=true
    SpriteSize=&Multiply_Vector2D_ByFloat(
     Float=&FloatFromCurve(
      FloatCurve=DataInterface() {propertyValues {Curve {keys [
          {arriveTangent=2.5229, interpMode=RCIM_Cubic, leaveTangent=2.5229, tangentMode=RCTM_User, time=0, value=0}
          {arriveTangent=0.3271, interpMode=RCIM_Cubic, leaveTangent=0.3271, tangentMode=RCTM_User, time=1, value=1}
         ]}}}
      Scale Curve=&FN_FrontEndFloatModifier(Gameplay=450, Front End=300)
     )
     Vector2D=<1 1>
    )
    Write to Module Namespace=true
   ) {module=FN_Sprite_Attributes004}
  ]
  rendererClasses [&NiagaraSpriteRendererProperties]
  renderers [
   {rendererClass=&NiagaraSpriteRendererProperties}
   {rendererClass=&NiagaraSpriteRendererProperties, rendererIndex=1}
   {rendererClass=&NiagaraSpriteRendererProperties, rendererIndex=2}
  ]
 }
}
summary {
 systemName=NS_Emote_WordUnity_Transformation_Return
 emitters [
  {emitterName=NE_Flare, rendererClasses [&NiagaraSpriteRendererProperties]}
  {emitterName=NE_Decals, rendererClasses [&NiagaraSpriteRendererProperties]}
 ]
 userVariables [
  User.bIsFrontEnd {type=NiagaraBool}
  User.bIsFrontEndPreview {type=NiagaraBool}
 ]
}
